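# Material function that layers an animated monitor-static screen effect over input material attributes within a channel-selected mask.
# Three panning samples of a static texture form noise, combined with a rolling sweep, scanlines and a brightness-boosted image (parameter texture or mask channel), then tinted by Screen Tone.
# Noise and sweep also produce UV distortion, and one scanline layer fades by distance.
# The result lerps into emissive, with an optional static-switched blend of surface attributes.
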MaterialFunction /Game/Global/Materials/MaterialGym/Materials/MonitorStatic/MF_MonitorStatic
graph {nodes [
  n1 = GetMaterialAttributes(None=NamedRerouteUsage()) {attributeGetTypes [
    #1
    #2
    #3
    #4
    #5
   ]}
  n2 = SetMaterialAttributes!(
   MaterialAttributes=@n1.MaterialAttributes
   BaseColor=VectorParameter() {group=MonitorStatic, parameterName=BaseColor, defaultValue=<rgba 0 0 0 1>}.RGB
   Metallic=ScalarParameter() {group=MonitorStatic, parameterName=Metallic}
   Specular=ScalarParameter() {defaultValue=1, group=MonitorStatic, parameterName=Specular}
   Roughness=ScalarParameter() {defaultValue=0.2, group=MonitorStatic, parameterName=Roughness}
  ) {attributeSetTypes [
    #2
    #3
    #5
    #4
   ]}
  n3 = BlendMaterialAttributes(
   A=@n1.MaterialAttributes
   B=@n2
   Alpha=Multiply(A=ScalarParameter() {defaultValue=1, group=MonitorStatic, parameterName=MaterialAttributesBlend}, B=NamedRerouteUsage())
  )
  n4 = NamedRerouteUsage()
  n5 = NamedRerouteUsage()
  FunctionOutput(None=SetMaterialAttributes!(
    MaterialAttributes=StaticSwitchParameter(True=@n3, False=@n1.MaterialAttributes) {group=MonitorStatic, parameterName=BlendMaterialAttributes}
    EmissiveColor=LinearInterpolate(A=@n1.EmissiveColor, B=@n4, Alpha=@n5)
   ) {attributeSetTypes [#1]})
  NamedRerouteDeclaration(None=ScalarParameter() {defaultValue=1, group=MonitorStatic, parameterName=StaticSpeed}) {name=StaticSpeed, nodeColor=<rgba 0 1 0.6484 1>}
  n6 = FunctionInput() {bUsePreviewValueAsDefault=true, inputName=Mask, sortPriority=1, previewValue=<1 1 1 1>}
  NamedRerouteDeclaration(None=ChannelMaskParameter(None=@n6) {group=MonitorStatic, parameterName=MaskChannel}.R) {name=IN_Alpha, nodeColor=<rgba 0.875 1 0 1>}
  n7 = NamedRerouteUsage()
  n8 = Constant2Vector() {g=-0.05, r=6}
  n9 = NamedRerouteUsage()
  n10 = TextureSample(UVs=Panner(Coordinate=@n7, Speed=Multiply(A=@n8, B=@n9)) {speedX=6, speedY=-0.05}) {
   samplerSource=SSM_Wrap_WorldGroupSettings
   samplerType=SAMPLERTYPE_Masks
   texture=/Game/Missions/Primary/RepairTheShelter/Textures/Cinematic_Screen_Static
  }
  n11 = Constant2Vector() {g=-0.03, r=-3}
  n12 = NamedRerouteUsage()
  n13 = TextureSample(UVs=Panner(Coordinate=@n7, Speed=Multiply(A=@n11, B=@n12)) {speedX=-3, speedY=-0.03}) {
   samplerSource=SSM_Wrap_WorldGroupSettings
   samplerType=SAMPLERTYPE_Masks
   texture=&Cinematic_Screen_Static
  }
  n14 = Multiply(A=@n10.G, B=@n13.R)
  n15 = ScalarParameter() {defaultValue=0.07, group=MonitorStatic, parameterName=Static UV Distort}
  NamedRerouteDeclaration(None=AppendVector(A=Multiply(A=@n14, B=@n15), B=Constant())) {name=StaticUVDistrotion, nodeColor=<rgba 1 0.5625 0 1>}
  n16 = Multiply(A=ComponentMask(None=NamedRerouteUsage()) {g=true}) {constB=0.75}
  n17 = Multiply(A=Time()) {constB=-0.2}
  n18 = Power(Base=Frac(None=Add(A=@n16, B=@n17))) {constExponent=20}
  NamedRerouteDeclaration(None=Multiply(A=@n18, B=ScalarParameter() {defaultValue=0.2, group=MonitorStatic, parameterName=Sweep Add}) {constB=0.2}) {name=Sweep, nodeColor=<rgba 0.2422 0 1 1>}
  n19 = ScalarParameter() {defaultValue=0.005, group=MonitorStatic, parameterName=Sweep UV Distort}
  NamedRerouteDeclaration(None=AppendVector(A=Multiply(A=@n18, B=@n19) {constB=0.02}, B=Constant())) {name=SweepUVDistortion, nodeColor=<rgba 0.0781 1 0 1>}
  NamedRerouteDeclaration(None=FunctionInput(Preview=SetMaterialAttributes!(MaterialAttributes)) {bUsePreviewValueAsDefault=true, inputName=Material, inputType=FunctionInput_MaterialAttributes}) {name=IN_Material, nodeColor=<rgba 0.875 1 0 1>}
  n20 = TextureCoordinate()
  n21 = ComponentMask(None=VectorParameter() {group=MonitorStatic, parameterName=TexCoordScale, defaultValue=<rgba 1 1 0 0>}.RGB) {g=true, r=true}
  n22 = VectorParameter() {group=MonitorStatic, parameterName=TexCoordOffset}
  NamedRerouteDeclaration(None=Add(A=Multiply(A=@n20, B=@n21), B=ComponentMask(None=@n22.RGB) {g=true, r=true})) {name=TexCoord, nodeColor=<rgba 1 0 0.4453 1>}
  NamedRerouteDeclaration(None=ChannelMaskParameter(None=@n6) {group=MonitorStatic, maskChannel=Green, parameterName=TextureChannel, defaultValue=<rgba 0 1 0 0>}.R) {name=IN_Texture, nodeColor=<rgba 0 1 0.25 1>}
  n23 = NamedRerouteUsage()
  n24 = ConstantBiasScale(None=Frac(None=Panner(Coordinate=@n23) {speedY=-1})) {bias=-0.5, scale=2}
  n25 = ComponentMask(None=Multiply(A=@n24, B=@n24)) {g=true}
  n26 = ScalarParameter() {defaultValue=0.5, group=MonitorStatic, parameterName=Scanline Strength}
  NamedRerouteDeclaration(None=Add(A=Multiply(A=@n25, B=@n26) {constB=0.1})) {name=Scanlines, nodeColor=<rgba 0 0.1562 1 1>}
  n27 = ScalarParameter() {defaultValue=50, group=MonitorStatic, parameterName=ScanlineDistanceFadeRange}
  n28 = Multiply(A=ScalarParameter() {defaultValue=200, group=MonitorStatic, parameterName=ScanlineDistanceFadeStart}) {constB=-1}
  NamedRerouteDeclaration(None=LinearInterpolate(
    A=MaterialFunctionCall!(ScanLine_Tiling_1, Scanline_PannerSpeed_1, ScanLine_Tiling, Scanline_PannerSpeed, UVs) {materialFunction=/Game/Weapons/WeaponSkins/MaterialFunction/ScanLines}.Result
    B=Constant() {r=0.5}
    Alpha=MaterialFunctionCall!(Blend Range=@n27, Start Offset=@n28) {materialFunction=/Engine/Functions/Engine_MaterialFunctions03/Particles/Distance_Blend}.Result
   )) {name=Scanlines_02, nodeColor=<rgba 0.4766 1 0 1>}
  n29 = NamedRerouteUsage()
  n30 = NamedRerouteUsage()
  n31 = NamedRerouteUsage()
  n32 = NamedRerouteUsage()
  n33 = Add(A=NamedRerouteUsage(), B=NamedRerouteUsage())
  n34 = TextureCoordinate()
  n35 = StaticSwitchParameter(
   True=NamedRerouteUsage()
   False=TextureSampleParameter2D(UVs=Add(A=@n33, B=@n34)) {group=MonitorStatic, parameterName=Texture, texture=/BRCosmetics/2dAssets/Banners/Season15/T-Banners-Icons-S15-CosmosA}.RGB
  ) {group=MonitorStatic, parameterName=UseMaskAsTexture}
  n36 = Add(
   A=Add(A=Add(A=@n30, B=@n31), B=ScalarParameter() {defaultValue=0.2, group=MonitorStatic, parameterName=BackgroundBrightness})
   B=Multiply(A=Multiply(A=@n32, B=@n35), B=ScalarParameter() {defaultValue=5, group=MonitorStatic, parameterName=TextureIntensity})
  )
  NamedRerouteDeclaration(None=Multiply(
    A=VectorParameter() {group=MonitorStatic, parameterName=Screen Tone, defaultValue=<rgba 0.0044 0.125 0.0345 0>}.RGB
    B=Multiply(A=@n29, B=@n36) {constA=1}
   )) {name=OUT_Emissive, nodeColor=<rgba 1 0 0.7969 1>}
  n37 = Constant2Vector() {g=0.1, r=5}
  n38 = NamedRerouteUsage()
  n39 = TextureSample(UVs=Panner(Coordinate=@n7, Speed=Multiply(A=@n37, B=@n38)) {speedX=5, speedY=0.1}) {
   samplerSource=SSM_Wrap_WorldGroupSettings
   samplerType=SAMPLERTYPE_Masks
   texture=&Cinematic_Screen_Static
  }
  n40 = Multiply(A=@n39.B) {constB=0.1}
  NamedRerouteDeclaration(None=Multiply(A=Add(A=@n40, B=@n14), B=ScalarParameter() {defaultValue=6, group=MonitorStatic, parameterName=Static Scale})) {name=Static, nodeColor=<rgba 0 1 0.8828 1>}
 ]}
