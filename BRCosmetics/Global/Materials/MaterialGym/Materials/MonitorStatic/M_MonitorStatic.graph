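# M_MonitorStatic is a Default Lit material for a monitor static effect, judging by its name.
# It uses material attributes and delegates its shading to the MF_MonitorStatic material function.
# That function receives a SetMaterialAttributes node as its material input alongside a Mask input, so the static logic lives in the function rather than in this graph.
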Material /BRCosmetics/Global/Materials/MaterialGym/Materials/MonitorStatic/M_MonitorStatic
graph {
 properties {bAutomaticallySetUsageInEditor=false, bUseMaterialAttributes=true, pixelDepthOffsetMode=PDOM_Legacy, usedShadingModels=Default Lit}
 nodes [MaterialFunctionCall!(Material=SetMaterialAttributes!(MaterialAttributes), Mask) {materialFunction=/Game/Global/Materials/MaterialGym/Materials/MonitorStatic/MF_MonitorStatic}]
}
